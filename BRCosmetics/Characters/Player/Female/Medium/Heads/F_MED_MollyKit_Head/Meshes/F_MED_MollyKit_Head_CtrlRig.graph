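# Control Rig for the Molly Kit female head cosmetic that shifts the whiskers in response to upper-arm poses.
# A spherical pose reader on each upper arm outputs a falloff weight from asymmetric, hand-oriented active regions.
# Each weight drives a Modify Transforms unit applying a small rotation and translation to the same-side dyn_whisker master bone, carrying its child whisker bones.
# Both readers have debug drawing enabled.
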ControlRigBlueprint /BRCosmetics/Characters/Player/Female/Medium/Heads/F_MED_MollyKit_Head/Meshes/F_MED_MollyKit_Head_CtrlRig
graphs [
 RigVMFunctionLibrary {links [], nodes []}
 RigVMModel {nodes [
   RigUnit_BeginExecution = &RigUnit_BeginExecution()
   n1 = &RigUnit_SphericalPoseReader(
    ActiveRegionScaleFactors=(PositiveWidth=2,NegativeWidth=2,PositiveHeight=0.5,NegativeHeight=2)
    ActiveRegionSize=0.15
    Debug=(bDrawDebug=True,bDraw2D=False,bDrawLocalAxes=False,DebugScale=25,DebugSegments=20,DebugThickness=0.25)
    DriverAxis=(X=-1,Y=0,Z=0)
    DriverItem=(Type=Bone,Name="upperarm_r")
    FalloffRegionScaleFactors=(PositiveWidth=1.5,NegativeWidth=1.5,PositiveHeight=0.1,NegativeHeight=1.5)
    FalloffSize=0.4
    FlipHeightScaling=False
    FlipWidthScaling=False
    OptionalParentItem=(Type=Bone,Name="None")
    RotationOffset=(X=-145,Y=10,Z=-30)
    ExecutePin=@RigUnit_BeginExecution.ExecutePin
   )
   > &RigUnit_ModifyTransforms(
    ItemToModify=((Item=(Type=Bone,Name="dyn_whisker_r_MSTR"),Transform=(Rotation=(X=0,Y=-0.0872,Z=0,W=0.9962),Translation=(X=0,Y=0,Z=0.8),Scale3D=(X=1,Y=1,Z=1))))
    WeightMinimum=0
    Weight=@n1.OutputParam
   )
   > n2 = &RigUnit_SphericalPoseReader(
    ActiveRegionScaleFactors=(PositiveWidth=0.5,NegativeWidth=2,PositiveHeight=2,NegativeHeight=2)
    ActiveRegionSize=0.15
    Debug=(bDrawDebug=True,bDraw2D=False,bDrawLocalAxes=False,DebugScale=25,DebugSegments=20,DebugThickness=0.25)
    DriverAxis=(X=1,Y=0,Z=0)
    DriverItem=(Type=Bone,Name="upperarm_l")
    FalloffRegionScaleFactors=(PositiveWidth=0.1,NegativeWidth=1.5,PositiveHeight=1.5,NegativeHeight=1.5)
    FalloffSize=0.4
    FlipHeightScaling=False
    FlipWidthScaling=False
    OptionalParentItem=(Type=Bone,Name="None")
    RotationOffset=(X=-5,Y=-40,Z=50)
   )
   > &RigUnit_ModifyTransforms(
    ItemToModify=((Item=(Type=Bone,Name="dyn_whisker_l_MSTR"),Transform=(Rotation=(X=0,Y=-0.0872,Z=0,W=0.9962),Translation=(X=0,Y=0,Z=0.8),Scale3D=(X=1,Y=1,Z=1))))
    WeightMinimum=0
    Weight=@n2.OutputParam
   )
  ]}
]
hierarchy [
 root {loc=<0 0 0>, rot=<0 0 0 1>}
 pelvis {parent=root, loc=<0 -0.5301 98.0571>, rot=<0.0127 0.707 -0.0127 -0.707>}
 spine_01 {parent=pelvis, loc=<6.1208 -0.9328 0>, rot=<0 0 0.034 -0.9994>}
 spine_02 {parent=spine_01, loc=<6.5997 -0.2095 0>, rot=<0 0 0.0238 -0.9997>}
 spine_03 {parent=spine_02, loc=<6.4914 -0.1237 0>, rot=<0 0 0.0149 -0.9999>}
 spine_04 {parent=spine_03, loc=<6.9598 0.0794 0>, rot=<0 0 -0.0132 -0.9999>}
 spine_05 {parent=spine_04, loc=<6.9987 0 0>, rot=<0 0 -0.0015 -1>}
 neck_01 {parent=spine_05, loc=<11.3138 0.1529 0>, rot=<0 0 -0.2003 -0.9797>}
 neck_02 {parent=neck_01, loc=<4.5627 0 0>, rot=<0 0 0.0504 -0.9987>}
 head {parent=neck_02, loc=<5.3914 0 0>, rot=<0 0 0.115 -0.9934>}
 faceAttach {parent=head, loc=<0 0 0>, rot=<0.0029 -0.7071 -0.0029 -0.7071>}
 dyn_ear_r {parent=head, loc=<16.4379 0.1184 6.1414>, rot=<-0.5 -0.5 -0.5 -0.5>}
 dyn_ear_l {parent=head, loc=<16.4379 0.1184 -6.1414>, rot=<-0.5 -0.5 -0.5 -0.5>}
 dyn_whisker_r_MSTR {parent=head, loc=<-0.4415 0.8813 9.8772>, rot=<0.7071 0 0.7071 0>}
 dyn_whisker_r_upper {parent=dyn_whisker_r_MSTR, loc=<3.4863 0 1.2768>, rot=<0 0 0 1>}
 dyn_whisker_r_low {parent=dyn_whisker_r_MSTR, loc=<2.8393 0.2183 -2.8228>, rot=<0 0 0 1>}
 dyn_whisker_l_MSTR {parent=head, loc=<-0.4415 0.8813 -9.8772>, rot=<0 -0.7071 0 -0.7071>}
 dyn_whisker_l_upper {parent=dyn_whisker_l_MSTR, loc=<3.4862 0 1.277>, rot=<0 0 0 1>}
 dyn_whisker_l_low {parent=dyn_whisker_l_MSTR, loc=<2.8393 -0.2183 -2.8223>, rot=<0 0 0 1>}
 clavicle_l {parent=spine_05, loc=<5.4831 -0.1053 -3.2946>, rot=<0.0465 -0.7289 -0.009 -0.6829>}
 upperarm_l {parent=clavicle_l, loc=<9.8658 -1.2136 -0.4105>, rot=<0.0164 -0.3245 -0.0269 -0.9454>}
 clavicle_r {parent=spine_05, loc=<5.4831 -0.1053 3.2946>, rot=<0.7289 0.0465 -0.6829 0.009>}
 upperarm_r {parent=clavicle_r, loc=<-9.8658 1.2136 0.4105>, rot=<0.0164 -0.3245 -0.0269 -0.9454>}
 C_glabella_down_pose {type=Curve, loc=<0 0 0>, rot=<0 0 0 1>}
 C_glabella_up_pose {type=Curve, loc=<0 0 0>, rot=<0 0 0 1>}
 jaw_open_pose {type=Curve, loc=<0 0 0>, rot=<0 0 0 1>}
 L_blink_pose {type=Curve, loc=<0 0 0>, rot=<0 0 0 1>}
 L_upper_lip_raiser_pose {type=Curve, loc=<0 0 0>, rot=<0 0 0 1>}
 Phoneme_mbp_pose {type=Curve, loc=<0 0 0>, rot=<0 0 0 1>}
 phoneme_oo_pose {type=Curve, loc=<0 0 0>, rot=<0 0 0 1>}
 R_blink_pose {type=Curve, loc=<0 0 0>, rot=<0 0 0 1>}
 R_brow_down_pose {type=Curve, loc=<0 0 0>, rot=<0 0 0 1>}
 R_brow_up_pose {type=Curve, loc=<0 0 0>, rot=<0 0 0 1>}
 R_Frown_pose {type=Curve, loc=<0 0 0>, rot=<0 0 0 1>}
 R_lip_corner_narrow_pose {type=Curve, loc=<0 0 0>, rot=<0 0 0 1>}
 R_smile_pose {type=Curve, loc=<0 0 0>, rot=<0 0 0 1>}
 R_squint_inner_pose {type=Curve, loc=<0 0 0>, rot=<0 0 0 1>}
 R_upper_lilp_raiser_pose {type=Curve, loc=<0 0 0>, rot=<0 0 0 1>}
]
