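# Post-process AnimBlueprint for the MollyKit female head cosmetic, driving its face and secondary motion for whiskers and ears.
# It copies pose and curves from the attached parent mesh, layers in the BuffCat male facial sequence weighted by the DisableFaceOverride curve, then runs the head's Control Rig.
# Chained head-relative Anim Dynamics nodes then simulate four whiskers and two ears, with angular springs and simulation-space gravity on the ears.
# The event graph holds no logic.
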AnimBlueprint /BRCosmetics/Characters/Player/Female/Medium/Heads/F_MED_MollyKit_Head/Meshes/F_MED_MollyKit_Head_PostProcess_AnimBlueprint
target_skeleton=&/F_MED_MollyKit_Head_PostProcess
graphs [
 AnimGraph {nodes [
   n1 = GetCurveValue(CurveName=DisableFaceOverride)
   n2 = Control Rig(Source=Layered blend per bone(
     BasePose=Copy Pose From Mesh() {bCopyCurves=true, bUseAttachedParent=true}
     BlendPoses_0=Sequence Player 'BuffCat_Male_Facial'() {sequence=/BRCosmetics/Animation/Game/MainPlayer/Facial/BuffCat/BuffCat_Male_Facial}
     BlendWeights_0=@n1
    )) {
    alphaBoolBlend {blendOption=Linear}
    controlRigAssetReference {blueprintRigClass=&/F_MED_MollyKit_Head_CtrlRig.F_MED_MollyKit_Head_CtrlRig_C}
   }
   n3 = Anim Dynamics - Bone: dyn_whisker_r_upper(ComponentPose=Local To Component(LocalPose=@n2), Alpha=1, ExternalForce="0,0,0") {
    angularDampingOverride=1
    bOverrideAngularDamping=true
    bUseGravityOverride=true
    simulationSpace=BoneRelative
    alphaBoolBlend {blendOption=Linear}
    boundBone {boneName=dyn_whisker_r_upper}
    componentAppliedLinearAccClamp=<2000 2000 2000>
    componentLinearAccScale=<0.1 0.1 0.1>
    componentLinearVelScale=<0.1 0.1 0.1>
    gravityOverride {z=-400}
    physicsBodyDefinitions [{
      boundBone {boneName=dyn_whisker_r_upper}
      boxExtents=<20 20 20>
      constraintSetup {angularLimitsMax=<2 10 20>, angularLimitsMin=<-2 -10 -20>, angularTarget=<0 0 0>}
      localJointOffset=<5 0 0>
     }]
    relativeSpaceBone {boneName=head}
   }
   n4 = Anim Dynamics - Bone: dyn_whisker_r_low(ComponentPose=@n3, Alpha=1, ExternalForce="0,0,0") {
    angularDampingOverride=1
    bOverrideAngularDamping=true
    bUseGravityOverride=true
    simulationSpace=BoneRelative
    alphaBoolBlend {blendOption=Linear}
    boundBone {boneName=dyn_whisker_r_low}
    componentAppliedLinearAccClamp=<2000 2000 2000>
    componentLinearAccScale=<0.1 0.1 0.1>
    componentLinearVelScale=<0.1 0.1 0.1>
    gravityOverride {z=-400}
    physicsBodyDefinitions [{
      boundBone {boneName=dyn_whisker_r_low}
      boxExtents=<20 20 20>
      constraintSetup {angularLimitsMax=<2 10 20>, angularLimitsMin=<-2 -10 -20>, angularTarget=<0 0 0>}
      localJointOffset=<5 0 0>
     }]
    relativeSpaceBone {boneName=head}
   }
   n5 = Anim Dynamics - Bone: dyn_whisker_l_upper(ComponentPose=@n4, Alpha=1, ExternalForce="0,0,0") {
    angularDampingOverride=1
    bOverrideAngularDamping=true
    bUseGravityOverride=true
    simulationSpace=BoneRelative
    alphaBoolBlend {blendOption=Linear}
    boundBone {boneName=dyn_whisker_l_upper}
    componentAppliedLinearAccClamp=<2000 2000 2000>
    componentLinearAccScale=<0.1 0.1 0.1>
    componentLinearVelScale=<0.1 0.1 0.1>
    gravityOverride {z=-400}
    physicsBodyDefinitions [{
      boundBone {boneName=dyn_whisker_l_upper}
      boxExtents=<20 20 20>
      constraintSetup {angularLimitsMax=<2 10 20>, angularLimitsMin=<-2 -10 -20>, angularTarget=<0 0 0>}
      localJointOffset=<5 0 0>
     }]
    relativeSpaceBone {boneName=head}
   }
   n6 = Anim Dynamics - Bone: dyn_whisker_l_low(ComponentPose=@n5, Alpha=1, ExternalForce="0,0,0") {
    angularDampingOverride=1
    bOverrideAngularDamping=true
    bUseGravityOverride=true
    simulationSpace=BoneRelative
    alphaBoolBlend {blendOption=Linear}
    boundBone {boneName=dyn_whisker_l_low}
    componentAppliedLinearAccClamp=<2000 2000 2000>
    componentLinearAccScale=<0.1 0.1 0.1>
    componentLinearVelScale=<0.1 0.1 0.1>
    gravityOverride {z=-400}
    physicsBodyDefinitions [{
      boundBone {boneName=dyn_whisker_l_low}
      boxExtents=<20 20 20>
      constraintSetup {angularLimitsMax=<2 10 20>, angularLimitsMin=<-2 -10 -20>, angularTarget=<0 0 0>}
      localJointOffset=<5 0 0>
     }]
    relativeSpaceBone {boneName=head}
   }
   n7 = Anim Dynamics - Bone: dyn_ear_r(ComponentPose=@n6, Alpha=1, ExternalForce="0,0,0") {
    angularDampingOverride=1
    angularSpringConstant=5000
    bAngularSpring=true
    bGravityOverrideInSimSpace=true
    bOverrideAngularDamping=true
    bUseGravityOverride=true
    simulationSpace=BoneRelative
    alphaBoolBlend {blendOption=Linear}
    boundBone {boneName=dyn_ear_r}
    componentAppliedLinearAccClamp=<2000 2000 2000>
    componentLinearAccScale=<0.2 0.2 0.2>
    componentLinearVelScale=<0.2 0.2 0.2>
    gravityOverride {x=980}
    physicsBodyDefinitions [{
      boundBone {boneName=dyn_ear_r}
      boxExtents=<20 20 20>
      constraintSetup {angularLimitsMax=<4 5 2.5>, angularLimitsMin=<-4 -15 -2.5>, angularTarget=<0 0 0>}
      localJointOffset=<0 0 5>
     }]
    relativeSpaceBone {boneName=head}
   }
   n8 = Anim Dynamics - Bone: dyn_ear_l(ComponentPose=@n7, Alpha=1, ExternalForce="0,0,0") {
    angularDampingOverride=1
    angularSpringConstant=5000
    bAngularSpring=true
    bGravityOverrideInSimSpace=true
    bOverrideAngularDamping=true
    bUseGravityOverride=true
    simulationSpace=BoneRelative
    alphaBoolBlend {blendOption=Linear}
    boundBone {boneName=dyn_ear_l}
    componentAppliedLinearAccClamp=<2000 2000 2000>
    componentLinearAccScale=<0.2 0.2 0.2>
    componentLinearVelScale=<0.2 0.2 0.2>
    gravityOverride {x=980}
    physicsBodyDefinitions [{
      boundBone {boneName=dyn_ear_l}
      boxExtents=<20 20 20>
      constraintSetup {angularLimitsMax=<4 5 2.5>, angularLimitsMin=<-4 -15 -2.5>, angularTarget=<0 0 0>}
      localJointOffset=<0 0 5>
     }]
    relativeSpaceBone {boneName=head}
   }
   Output Pose(Result=Component To Local(ComponentPose=@n8))
   float * float(A=1, B=@n1)
  ]}
 EventGraph {nodes [
   TryGetPawnOwner()
   Event BlueprintUpdateAnimation()
  ]}
]
variables []
